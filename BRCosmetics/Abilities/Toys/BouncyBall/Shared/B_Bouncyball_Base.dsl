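; Base Blueprint for the bouncy ball toy, a thrown projectile that counts bounces and reports them.
; Hits above a speed threshold mirror and scale velocity and gravity, then play cosmetics or, where cosmetics cannot run, increment the count.
; On stop, it sends a replicated struct whose OnRep prints thrower and bounces to clients' message feeds, and authority credits a 25-bounce quest.
; Construction randomizes the replicated material index; overlaps play a throttled bounce sound.
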(event Custom|QuestCheck
  (bind _returnvalue (Actor|GetOwner))
  (bind _returnvalue_2 (Fortnite|GetFortPlayerControllerfromActor _returnvalue))
  ;[1 1 1] Check number of bounces for quest
  (Networking|SwitchHasAuthority
    (:Authority
      (if (and (Variables|Default|GetCheckQuest) (>= (Variables|Default|GetBounces) 25))
        (Utilities|IsValid _returnvalue_2
          (:Is%20Valid
            (Quest|SendComplexCustomStatEvent (Quests|GetQuestManager _returnvalue_2 "Invalid") 0 "(GameplayTags=)" "(GameplayTags=((TagName=\"Athena.Quests.BouncyBallBounces\")))" 1)
            (Variables|Default|SetCheckQuest false)))))))

(event EventProjectileStop
  (params (ImpactResult :type Hit%20Result%20Structure%20%28by%20ref%29))
  (|CallParentFunction ImpactResult)
  (if (not (Variables|Default|GetPrintedMessage))
    (CallFunction|CreateMessage (Utilities|Casting|CastToFortPlayerState (Actor|GetOwner)) (Variables|Default|GetBounces))
    (CallFunction|QuestCheck)))

(event EventOnPawnTouchedMe
  (params (TouchingPawn :type Pawn%20Object%20Reference) (ImpactNormal :type Vector :default "0, 0, 0") (OtherVelocity :type Vector :default "0, 0, 0"))
  (|CallParentFunction TouchingPawn ImpactNormal OtherVelocity))

(event Collision|EventHit
  (params (MyComp :type Primitive%20Component%20Object%20Reference) (Other :type Actor%20Object%20Reference) (OtherComp :type Primitive%20Component%20Object%20Reference) (bSelfMoved :type Boolean) (HitLocation :type Vector :default "0, 0, 0") (HitNormal :type Vector :default "0, 0, 0") (NormalImpulse :type Vector :default "0, 0, 0") (Hit :type Hit%20Result%20Structure%20%28by%20ref%29))
  (bind _projectilemovement (Variables|Default|GetProjectileMovement))
  (bind _projectilemovement_2 (Variables|Default|GetProjectileMovement))
  (|CallParentFunction MyComp Other OtherComp bSelfMoved HitLocation HitNormal NormalImpulse Hit)
  (if (>= (Math|Vector|VectorLength (Variables|/Script/Engine.MovementComponent|GetVelocity _projectilemovement)) 8e+02)
    (Variables|/Script/Engine.ProjectileMovementComponent|SetProjectileGravityScale (* (Variables|/Script/Engine.ProjectileMovementComponent|GetProjectileGravityScale _projectilemovement) (Variables|Default|GetGravityMultiplier)) _projectilemovement_2)
    (Variables|/Script/Engine.MovementComponent|SetVelocity (* (Math|Vector|MirrorVectorbyNormal (Variables|/Script/Engine.MovementComponent|GetVelocity _projectilemovement) HitNormal) (Variables|Default|GetVelocityMultiplier)) _projectilemovement_2)
    (Networking|CanExecuteCosmeticEvents
      (:True
        (Effects|SpawnEmitteratLocation (Variables|Components|GetBounceFX) (Transformation|GetActorLocation) (Math|Rotator|MakeRotfromX (* (Math|Vector|Normalize (Transformation|GetVelocity)) -1)))
        (Audio|Components|Audio|Play (Variables|Default|GetImpactSFX) 0.0))
      (:False
        (Variables|Default|SetBounces (+ (Variables|Default|GetBounces) 1))))))

(event EventSetBallVelocityAndStartTicking
  (params (NewVelocity :type Vector :default "0, 0, 0"))
  (|CallParentFunction NewVelocity)
  (CallFunction|ClientonThrown))

(event Custom|Client%20On%20Thrown
  (Utilities|FlowControl|DoOnce false
    (:Completed
      (Variables|Default|SetIs%20Throwing false)
      ;[1 1 1] Do we need to check quest credit?
      (Networking|SwitchHasAuthority))))

(event Collision|EventActorBeginOverlap
  (params (OtherActor :type Actor%20Object%20Reference))
  (|CallParentFunction OtherActor)
  (CallFunction|HitPlayer/VehicleBounceSound))

(event EventBeginPlay
  (|CallParentFunction)
  (Variables|Default|SetIs%20Throwing true))

(event Custom|Hit%20Player/Vehicle%20Bounce%20Sound
  ;[1 1 1] Support Collision with Players and Vehicles
  (if (not (Variables|Default|GetIs%20Throwing))
    (Utilities|FlowControl|DoOnce false
      (:Completed
        (Audio|PlaySoundatLocation "/BRCosmetics/Sounds/Other/Toys/BouncyBall/Toys_BouncyBall_Bounce_Cue.Toys_BouncyBall_Bounce_Cue" (Transformation|GetActorLocation))
        (Utilities|FlowControl|Delay 0.2)))))

(event EventTick
  (params (DeltaSeconds :type Float%20%28single-precision%29))
  (|CallParentFunction DeltaSeconds) ; This node is disabled and will not be called.\nDrag off pins to build functionality.
)

(fn CreateMessage
  (params (ThrowingPlayer :type Fort%20Player%20State%20Object%20Reference) (Bounces :type Integer))
  (if (Variables|/Script/FortniteGame.FortRuntimeOptions|GetbToyMessagingEnabled (Options|GetRuntimeOptions))
    (Variables|Default|SetDisplayMessage (Utilities|Struct|MakeStructBouncyballMinigameMessage ThrowingPlayer Bounces (Variables|Default|GetMessageSN)))
    (Variables|Default|SetMessageSN (Math|Byte|Byte+Byte (Variables|Default|GetMessageSN) 1))))

(fn OnRep_Bounces)

(fn OnRep_CheckQuest)

(fn OnRep_DisplayMessage
  (bind _returnvalue (BlueprintContext|GetFortToyContext))
  (bind (_displaymessage_playerwhothrew_22_ef5d780b4020e911ab429d9ba6b977ba _displaymessage_bounces_28_a600d2a048cb01620b914bb4107a6cd5) (Utilities|Struct|BreakStructBouncyballMinigameMessage (Variables|GetDisplayMessage_PlayerWhoThrew_22_EF5D780B4020E911AB429D9BA6B977BA)))
  (Networking|SwitchHasAuthority
    (:Remote
      (Utilities|IsValid _displaymessage_playerwhothrew_22_ef5d780b4020e911ab429d9ba6b977ba
        (:Is%20Valid
          (bind _returnvalue_2 (Class|FortToyContext|FormatPlayerNameforDisplay (BlueprintContext|GetFortToyContext) _displaymessage_playerwhothrew_22_ef5d780b4020e911ab429d9ba6b977ba "ToyOwner"))
          (bind _formatstring (CallFunction|PickFormatString))
          (Utilities|IsValid _returnvalue
            (:Is%20Valid
              (if (not (Variables|Default|GetPrintedMessage))
                (Class|FortToyContext|ConditionalPrinttoMessageFeed _returnvalue _displaymessage_playerwhothrew_22_ef5d780b4020e911ab429d9ba6b977ba (Transformation|GetActorLocation) (select (> (Math|Vector|VectorLength (Transformation|GetActorLocation)) (Variables|Default|GetMinDistanceForTeamRelevance)) "SquadAndTeamMembers" "SquadMembersOnly") (Utilities|Text|FormatText _formatstring _returnvalue_2 _displaymessage_bounces_28_a600d2a048cb01620b914bb4107a6cd5))
                (Variables|Default|SetPrintedMessage true)))))))))

(fn OnRep_MaterialIndex
  (Rendering|Material|SetMaterial (Variables|Default|GetBall) 0 (Utilities|Select :Index (Variables|Default|GetMaterialIndex) :Option%200 (Variables|Materials|GetMaterial1) :Option%201 (Variables|Materials|GetMaterial2) :Option%202 (Variables|Materials|GetMaterial3)))
  (Development|PrintString (Utilities|String|ToString(Integer) (Variables|Default|GetMaterialIndex))))

(fn PickFormatString
  (return (Variables|Default|GetFormatString)))

(fn ConstructionScript
  (|CallParentFunction)
  (Networking|CanExecuteCosmeticEvents
    (:False
      (Variables|Default|SetMaterialIndex (Math|Random|RandomIntegerinRange 0 2)))))
